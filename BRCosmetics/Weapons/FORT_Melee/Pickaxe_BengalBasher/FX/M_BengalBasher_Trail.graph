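# Translucent unlit material for Niagara ribbon trails on the Bengal Basher pickaxe.
# Panning cloud normal noise distorts UVs for a blurred noise texture, which linear U and V gradients fade.
# A narrow SmoothStep sharpens that into a cutout shape, multiplied by a parameterized trailing-edge sphere mask and particle alpha.
# Particle color drives emissive through MF_FX_EmissiveAdjustments.
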Material /BRCosmetics/Weapons/FORT_Melee/Pickaxe_BengalBasher/FX/M_BengalBasher_Trail
graph {
 properties {bUsedWithNiagaraRibbons=true, blendMode=BLEND_Translucent, pixelDepthOffsetMode=PDOM_Legacy, shadingModel=MSM_Unlit, usedShadingModels=Unlit}
 nodes [
  n1 = ParticleColor()
  MaterialFunctionCall(Emissive=@n1.RGB) {materialFunction=&MF_FX_EmissiveAdjustments}
  n2 = Constant() {r=0.4}
  n3 = Add(A=@n2, B=Constant() {r=0.2})
  n4 = TextureCoordinate()
  n5 = TextureSample(UVs=Panner(Coordinate=TextureCoordinate()) {speedX=0.1}) {samplerType=SAMPLERTYPE_Normal, texture=/Game/Accessories/FORT_Backpacks/Backpack_M_MED_EctoCat/FX/CloudNoise_T_N}
  n6 = MaterialFunctionCall(X=@n5.R, Y=@n5.R) {materialFunction=&MakeFloat2}
  n7 = ScalarParameter() {defaultValue=0.1, parameterName=UV Distortion Amount}
  n8 = VectorParameter() {parameterName=UV Scale 1, defaultValue=<rgba 0.5 3 0 0>}
  n9 = Multiply(
   A=Add(A=@n4, B=Multiply(A=@n6.Result, B=@n7))
   B=MaterialFunctionCall(X=@n8.R, Y=@n8.R) {materialFunction=&MakeFloat2}.Result
  )
  n10 = Multiply(A=OneMinus(None=ComponentMask(None=@n4) {r=true})) {constB=1.25}
  n11 = OneMinus(None=ComponentMask(None=@n4) {g=true})
  n12 = Constant() {r=3}
  n13 = Multiply(
   A=Multiply(
    A=TextureSample(UVs=@n9) {samplerType=SAMPLERTYPE_Grayscale, texture=/FortUILibrary/UIKit/Buttons/Textures/LowResBlurredNoise}.R
    B=Saturate(None=@n10)
   )
   B=Saturate(None=Multiply(A=@n11, B=@n12))
  )
  n14 = VectorParameter() {group=Trailing Edge Sphere Mask, parameterName=Trailing Edge Sphereical UV Scale, sortPriority=4, defaultValue=<rgba 0.5 1 0 0>}
  n15 = Multiply(A=TextureCoordinate(), B=ComponentMask(None=@n14.RGB) {g=true, r=true})
  n16 = VectorParameter() {group=Trailing Edge Sphere Mask, parameterName="Trailing Edge Sphere Mask Coord (B)", sortPriority=3, defaultValue=<rgba 0.85 0.82 0 0>}
  n17 = ComponentMask(None=@n16.RGB) {g=true, r=true}
  n18 = ScalarParameter() {defaultValue=0.8, group=Trailing Edge Sphere Mask, parameterName=Trailing Edge Radius, sortPriority=1}
  n19 = ScalarParameter() {defaultValue=0.3, group=Trailing Edge Sphere Mask, parameterName=Trailing Edge Hardness, sortPriority=2}
  n20 = Power(
   Base=OneMinus(None=SphereMask(A=@n15, B=@n17, Radius=@n18, Hardness=@n19))
   Exp=ScalarParameter() {defaultValue=5, group=Trailing Edge Sphere Mask, parameterName=Trailing Edge Contrast}
  ) {constExponent=5}
  n21 = ScalarParameter() {defaultValue=1, group=Alpha, parameterName=Multiply Trailing Edge Sphere Mask, sortPriority=6}
  Multiply(A=@n1.A, B=Multiply(A=SmoothStep(Min=@n2, Max=@n3, Value=@n13), B=LinearInterpolate(B=@n20, Alpha=@n21) {constA=1}))
 ]
}
